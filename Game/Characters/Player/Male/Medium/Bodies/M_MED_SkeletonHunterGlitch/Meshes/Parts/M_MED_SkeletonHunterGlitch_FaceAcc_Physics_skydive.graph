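# Skydive physics asset for the SkeletonHunterGlitch player face accessory, setting up dynamic hair motion.
# Capsules on upper spine, clavicles, neck and head form the body, with back, front, side and right-clavicle hair chains branching from a head-attached base.
# Body and most base joints are angularly free, while hair segments use asymmetric left/right swing limits and tight twist.
# An extra angularly free constraint cross-links the left front and side chains.
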PhysicsAsset /Game/Characters/Player/Male/Medium/Bodies/M_MED_SkeletonHunterGlitch/Meshes/Parts/M_MED_SkeletonHunterGlitch_FaceAcc_Physics_skydive
bodies {
 clavicle_l [{shapeType=Capsule, length=3.75, radius=10, center=<5.544 -1.2447 -1.9153>, rotation=<rot 90 0 0>}]
 clavicle_r [
  {shapeType=Capsule, length=25.0931, radius=6, center=<-15.2824 2.1135 -2.6383>, rotation=<rot 90 -0.0049 179.995>}
  {
   shapeType=Capsule
   length=7.0661
   radius=5.8224
   shapeName=clavicle_r_capsule
   center=<4.469 2.2495 -3.3245>
   rotation=<rot -90 180 179.9999>
  }
 ]
 dyn_hair_back_01 [{shapeType=Capsule, length=3, radius=5, shapeName=dyn_hair_back_01_capsule, center=<5 0 4>, rotation=<rot 90 0 0>}]
 dyn_hair_back_02 [{shapeType=Capsule, length=2.75, radius=6.25, shapeName=dyn_hair_back_02_capsule, center=<5 0 4>, rotation=<rot 85 0 0>}]
 dyn_hair_back_03 [{shapeType=Capsule, length=4.5, radius=6, shapeName=dyn_hair_back_03_capsule, center=<5 0 4>, rotation=<rot 85 0 0>}]
 dyn_hair_back_04 [{shapeType=Capsule, length=3.25, radius=6, shapeName=dyn_hair_back_04_capsule, center=<5 0 4>, rotation=<rot 80 0 0>}]
 dyn_hair_back_base [{shapeType=Capsule, length=0.505, radius=0.505, shapeName=dyn_hair_back_base_capsule, rotation=<rot 0 -90 0>}]
 dyn_hair_base [{shapeType=Capsule, length=0.505, radius=0.505, shapeName=dyn_hair_base_capsule, rotation=<rot 0 -90 0>}]
 dyn_hair_clavicle_side_r [{
   shapeType=Capsule
   length=20.7483
   radius=2.7336
   shapeName=dyn_hair_clavicle_side_r_capsule
   center=<7.8384 -2.8699 9.7483>
   rotation=<rot 0 -0.0001 -90>
  }]
 dyn_hair_clavicle_side_r_base [{shapeType=Capsule, length=0.505, radius=0.505, shapeName=dyn_hair_clavicle_side_r_base_capsule, rotation=<rot 0 -90 0>}]
 dyn_hair_front_01_l [{shapeType=Capsule, length=4, radius=3, shapeName=dyn_hair_front_01_l_capsule, center=<0 0 1.0904>, rotation=<rot 90 -178.2101 176.9018>}]
 dyn_hair_front_01_r [{shapeType=Capsule, length=2.75, radius=2.5, shapeName=dyn_hair_front_01_r_capsule, rotation=<rot 90 0 0>}]
 dyn_hair_front_02_l [{
   shapeType=Capsule
   length=4.25
   radius=2.75
   shapeName=dyn_hair_front_02_l_capsule
   center=<3.1082 0 0.4109>
   rotation=<rot -89.2974 179.9998 -179.9997>
  }]
 dyn_hair_front_02_r [{
   shapeType=Capsule
   length=4.5
   radius=6
   shapeName=dyn_hair_front_02_r_capsule
   center=<8.8823 -2.5499 1.5>
   rotation=<rot 90 40.4252 49.7422>
  }]
 dyn_hair_front_03_l [{
   shapeType=Capsule
   length=2.3995
   radius=2
   shapeName=dyn_hair_front_03_l_capsule
   center=<4.5612 0 0.061>
   rotation=<rot 89.2338 -179.9997 -179.9998>
  }]
 dyn_hair_front_03_r [{
   shapeType=Capsule
   length=17.5895
   radius=4
   shapeName=dyn_hair_front_03_r_capsule
   center=<11.5507 -6.2066 4.2927>
   rotation=<rot 74.9902 179.9999 -166.1175>
  }]
 dyn_hair_side_01_l [{shapeType=Capsule, length=3.25, radius=4.25, shapeName=dyn_hair_side_01_l_capsule, center=<6.1746 0 1.7929>, rotation=<rot 79.9996 0 0>}]
 dyn_hair_side_01_r [{shapeType=Capsule, length=2.75, radius=5, shapeName=dyn_hair_side_01_r_capsule, rotation=<rot 90 0 0>}]
 dyn_hair_side_02_l [{
   shapeType=Capsule
   length=3.375
   radius=5.875
   shapeName=dyn_hair_side_02_l_capsule
   center=<10.1178 1.7725 1>
   rotation=<rot -0.0001 20 89.9999>
  }]
 dyn_hair_side_02_r [{shapeType=Capsule, length=5, radius=8, shapeName=dyn_hair_side_02_r_capsule, center=<4.0841 0 -0.1168>, rotation=<rot 88.3623 0 0.0001>}]
 dyn_hair_side_03_l [{shapeType=Capsule, length=6, radius=5, shapeName=dyn_hair_side_03_l_capsule, center=<5 0 2>, rotation=<rot 80.0001 -179.9999 -180>}]
 dyn_hair_side_03_r [{shapeType=Capsule, length=14.6573, radius=3.7898, shapeName=dyn_hair_side_03_r_capsule, center=<14.354 3.9019 -2.531>, rotation=<rot 80 0 0>}]
 dyn_hair_side_l_base [{shapeType=Capsule, length=0.505, radius=0.505, shapeName=dyn_hair_side_l_base_capsule, rotation=<rot 0 -90 0>}]
 dyn_hair_side_r_base [{shapeType=Capsule, length=0.505, radius=0.505, shapeName=dyn_hair_side_r_base_capsule, rotation=<rot 0 -90 0>}]
 head [{shapeType=Capsule, length=4, radius=5.5, center=<2.4 2.1 0>, rotation=<rot 90 0.0148 0.0149>}]
 neck_01 [{shapeType=Capsule, length=3.75, radius=5.45, center=<3 0 0>, rotation=<rot 90 0 0>}]
 spine_03 [{shapeType=Capsule, length=20.25, radius=12.75, center=<4.4081 -0.938 0>, rotation=<rot -0.0006 0 0>}]
 spine_04 [
  {shapeType=Capsule, length=21.75, radius=11.25, center=<2.8061 -0.816 0>, rotation=<rot -0.0006 0 0>}
  {shapeType=Capsule, length=15, radius=15, shapeName=spine_04_capsule, center=<15.6918 0 0>, rotation=<rot 90 -179.9994 -179.9993>}
 ]
 spine_05 [{shapeType=Capsule, length=32.3258, radius=8.375, center=<2.7027 -1.4413 0.0001>, rotation=<rot -0.0006 0 0>}]
}
constraints [
 {
  bone1Name=spine_04
  bone2Name=spine_03
  swing1LimitDegrees=45
  swing1Motion=Free
  swing2LimitDegrees=45
  swing2Motion=Free
  twistLimitDegrees=45
  twistMotion=Free
 }
 {
  bone1Name=spine_05
  bone2Name=spine_04
  swing1LimitDegrees=45
  swing1Motion=Free
  swing2LimitDegrees=45
  swing2Motion=Free
  twistLimitDegrees=45
  twistMotion=Free
 }
 {
  bone1Name=clavicle_l
  bone2Name=spine_05
  swing1LimitDegrees=45
  swing1Motion=Free
  swing2LimitDegrees=45
  swing2Motion=Free
  twistLimitDegrees=45
  twistMotion=Free
 }
 {
  bone1Name=clavicle_r
  bone2Name=spine_05
  swing1LimitDegrees=45
  swing1Motion=Free
  swing2LimitDegrees=45
  swing2Motion=Free
  twistLimitDegrees=45
  twistMotion=Free
 }
 {
  bone1Name=neck_01
  bone2Name=spine_05
  swing1LimitDegrees=45
  swing1Motion=Free
  swing2LimitDegrees=45
  swing2Motion=Free
  twistLimitDegrees=45
  twistMotion=Free
 }
 {
  bone1Name=head
  bone2Name=neck_01
  swing1LimitDegrees=45
  swing1Motion=Free
  swing2LimitDegrees=45
  swing2Motion=Free
  twistLimitDegrees=45
  twistMotion=Free
 }
 {
  bone1Name=dyn_hair_base
  bone2Name=head
  swing1LimitDegrees=45
  swing1Motion=Free
  swing2LimitDegrees=45
  swing2Motion=Free
  twistLimitDegrees=45
  twistMotion=Free
 }
 {
  bone1Name=dyn_hair_back_base
  bone2Name=dyn_hair_base
  swing1LimitDegrees=45
  swing1Motion=Free
  swing2LimitDegrees=45
  swing2Motion=Free
  twistLimitDegrees=45
  twistMotion=Free
 }
 {bone1Name=dyn_hair_back_01, bone2Name=dyn_hair_back_base, swing1LimitDegrees=5, swing2LimitDegrees=45, twistLimitDegrees=2}
 {bone1Name=dyn_hair_back_02, bone2Name=dyn_hair_back_01, swing1LimitDegrees=10, swing2LimitDegrees=40, twistLimitDegrees=2}
 {bone1Name=dyn_hair_back_03, bone2Name=dyn_hair_back_02, swing1LimitDegrees=10, swing2LimitDegrees=40, twistLimitDegrees=2}
 {bone1Name=dyn_hair_back_04, bone2Name=dyn_hair_back_03, swing1LimitDegrees=10, swing2LimitDegrees=40, twistLimitDegrees=2}
 {bone1Name=dyn_hair_front_01_l, bone2Name=dyn_hair_base, swing1LimitDegrees=30, swing2LimitDegrees=30, twistLimitDegrees=2}
 {bone1Name=dyn_hair_front_02_l, bone2Name=dyn_hair_front_01_l, swing1LimitDegrees=30, swing2LimitDegrees=30, twistLimitDegrees=2}
 {bone1Name=dyn_hair_front_01_r, bone2Name=dyn_hair_base, swing1LimitDegrees=15, swing2LimitDegrees=15, twistLimitDegrees=2}
 {bone1Name=dyn_hair_front_02_r, bone2Name=dyn_hair_front_01_r, swing1LimitDegrees=15, swing2LimitDegrees=15, twistLimitDegrees=2}
 {bone1Name=dyn_hair_front_03_r, bone2Name=dyn_hair_front_02_r, swing1LimitDegrees=15, swing2LimitDegrees=15, twistLimitDegrees=2}
 {
  bone1Name=dyn_hair_side_l_base
  bone2Name=dyn_hair_base
  swing1LimitDegrees=45
  swing1Motion=Free
  swing2LimitDegrees=45
  swing2Motion=Free
  twistLimitDegrees=45
  twistMotion=Free
 }
 {bone1Name=dyn_hair_side_01_l, bone2Name=dyn_hair_side_l_base, swing1LimitDegrees=5, swing2LimitDegrees=45, twistLimitDegrees=2}
 {bone1Name=dyn_hair_side_02_l, bone2Name=dyn_hair_side_01_l, swing1LimitDegrees=10, swing2LimitDegrees=65, twistLimitDegrees=2}
 {bone1Name=dyn_hair_side_03_l, bone2Name=dyn_hair_side_02_l, swing1LimitDegrees=10, swing2LimitDegrees=65, twistLimitDegrees=2}
 {
  bone1Name=dyn_hair_side_r_base
  bone2Name=dyn_hair_base
  swing1LimitDegrees=45
  swing1Motion=Free
  swing2LimitDegrees=45
  swing2Motion=Free
  twistLimitDegrees=45
  twistMotion=Free
 }
 {bone1Name=dyn_hair_side_01_r, bone2Name=dyn_hair_side_r_base, swing1LimitDegrees=2, swing2LimitDegrees=10, twistLimitDegrees=2}
 {bone1Name=dyn_hair_side_02_r, bone2Name=dyn_hair_side_01_r, swing1LimitDegrees=10, swing2LimitDegrees=20, twistLimitDegrees=2}
 {bone1Name=dyn_hair_side_03_r, bone2Name=dyn_hair_side_02_r, swing1LimitDegrees=15, swing2LimitDegrees=15, twistLimitDegrees=2}
 {bone1Name=dyn_hair_front_03_l, bone2Name=dyn_hair_front_02_l, swing1LimitDegrees=30, swing2LimitDegrees=30, twistLimitDegrees=2}
 {
  bone1Name=dyn_hair_front_02_l
  bone2Name=dyn_hair_side_02_l
  swing1LimitDegrees=45
  swing1Motion=Free
  swing2LimitDegrees=45
  swing2Motion=Free
  twistLimitDegrees=45
  twistMotion=Free
 }
 {bone1Name=dyn_hair_clavicle_side_r, bone2Name=dyn_hair_clavicle_side_r_base, swing1LimitDegrees=5, swing2LimitDegrees=5, twistLimitDegrees=2}
 {bone1Name=dyn_hair_clavicle_side_r_base, bone2Name=dyn_hair_base, swing1LimitDegrees=45, swing2LimitDegrees=45, twistLimitDegrees=2}
]
