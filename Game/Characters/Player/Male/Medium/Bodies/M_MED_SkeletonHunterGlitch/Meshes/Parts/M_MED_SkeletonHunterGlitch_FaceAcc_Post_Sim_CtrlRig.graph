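# Post-sim Control Rig for the SkeletonHunterGlitch face accessory that adjusts the dynamic hair bones during skydiving and parachuting.
# While Skydiving is true, parent constraints tie hair bones to the head, with lower bones weighted across head, neck, spine and clavicle areas.
# A second sequence branch layers transform offsets weighted by interpolated skydive, left/right movement and parachute values.
# Spherical pose readers on both clavicles add further offsets, scaled by ParachuteOpen.
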ControlRigBlueprint /Game/Characters/Player/Male/Medium/Bodies/M_MED_SkeletonHunterGlitch/Meshes/Parts/M_MED_SkeletonHunterGlitch_FaceAcc_Post_Sim_CtrlRig
graphs [
 RigVMFunctionLibrary {links [], nodes []}
 RigVMModel {nodes [
   RigUnit_BeginExecution = &RigUnit_BeginExecution()
   RigVMFunction_Sequence = &RigVMFunction_Sequence(ExecuteContext=@RigUnit_BeginExecution.ExecutePin)
   RigVMFunction_ControlFlowBranch = &RigVMFunction_ControlFlowBranch(ExecuteContext=@RigVMFunction_Sequence.A, Condition=RigVMVariableNode() {variable=Skydiving}.Value)
   &RigUnit_ParentConstraint(
    AdvancedSettings=(RotationOrderForFilter=XZY)
    Child=(Type=Bone,Name="dyn_hair_front_01_l")
    Parents=((Item=(Type=Bone,Name="head"),Weight=1.0))
    Weight=1
    ExecutePin=@RigVMFunction_ControlFlowBranch.True
   )
   > &RigUnit_ParentConstraint(
    AdvancedSettings=(RotationOrderForFilter=XZY)
    Child=(Type=Bone,Name="dyn_hair_side_01_l")
    Parents=((Item=(Type=Bone,Name="head"),Weight=1.0))
    Weight=1
   )
   > &RigUnit_ParentConstraint(
    AdvancedSettings=(RotationOrderForFilter=XZY)
    Child=(Type=Bone,Name="dyn_hair_back_01")
    Parents=((Item=(Type=Bone,Name="head"),Weight=1.0))
    Weight=1
   )
   > &RigUnit_ParentConstraint(
    AdvancedSettings=(RotationOrderForFilter=XZY)
    Child=(Type=Bone,Name="dyn_hair_side_01_r")
    Parents=((Item=(Type=Bone,Name="head"),Weight=1.0))
    Weight=1
   )
   > &RigUnit_ParentConstraint(
    AdvancedSettings=(RotationOrderForFilter=XZY)
    Child=(Type=Bone,Name="dyn_hair_front_01_r")
    Parents=((Item=(Type=Bone,Name="head"),Weight=1.0))
    Weight=1
   )
   > &RigUnit_ParentConstraint(
    AdvancedSettings=(RotationOrderForFilter=XZY)
    Child=(Type=Bone,Name="dyn_hair_front_02_l")
    Parents=((Item=(Type=Bone,Name="head"),Weight=1.0))
    Weight=1
   )
   > &RigUnit_ParentConstraint(
    AdvancedSettings=(RotationOrderForFilter=XZY)
    Child=(Type=Bone,Name="dyn_hair_side_02_l")
    Parents=((Item=(Type=Bone,Name="head"),Weight=1.0))
    Weight=1
   )
   > &RigUnit_ParentConstraint(
    AdvancedSettings=(RotationOrderForFilter=XZY)
    Child=(Type=Bone,Name="dyn_hair_back_02")
    Parents=((Item=(Type=Bone,Name="head"),Weight=1.0))
    Weight=1
   )
   > &RigUnit_ParentConstraint(
    AdvancedSettings=(RotationOrderForFilter=XZY)
    Child=(Type=Bone,Name="dyn_hair_side_02_r")
    Parents=((Item=(Type=Bone,Name="head"),Weight=1.0))
    Weight=1
   )
   > &RigUnit_ParentConstraint(
    AdvancedSettings=(RotationOrderForFilter=XZY)
    Child=(Type=Bone,Name="dyn_hair_front_02_r")
    Parents=((Item=(Type=Bone,Name="head"),Weight=1.0))
    Weight=1
   )
   > &RigUnit_ParentConstraint(
    AdvancedSettings=(RotationOrderForFilter=XZY)
    Child=(Type=Bone,Name="dyn_hair_front_03_l")
    Parents="((Item=(Type=Bone,Name=\"clavicle_l\"),Weight=0.0101),(Item=(Type=Bone,Name=\"C_jaw\"),Weight=0.0336),(Item=(Type=Bone,Name=\"neck_01\"),Weight=0.1346),(Item=(Type=Bone,Name=\"spine_05\"),Weight=0.1803),(Item=(Type=Bone,Name=\"neck_02\"),Weight=0.3157),(Item=(Type=B ...(truncated)"
    Weight=1
   )
   > &RigUnit_ParentConstraint(
    AdvancedSettings=(RotationOrderForFilter=XZY)
    Child=(Type=Bone,Name="dyn_hair_side_03_l")
    Parents=((Item=(Type=Bone,Name="clavicle_l"),Weight=0.0471),(Item=(Type=Bone,Name="head"),Weight=0.0576),(Item=(Type=Bone,Name="neck_02"),Weight=0.1731),(Item=(Type=Bone,Name="neck_01"),Weight=0.3131),(Item=(Type=Bone,Name="spine_05"),Weight=0.409))
    Weight=1
   )
   > &RigUnit_ParentConstraint(
    AdvancedSettings=(RotationOrderForFilter=XZY)
    Child=(Type=Bone,Name="dyn_hair_back_04")
    Parents="((Item=(Type=Bone,Name=\"neck_01\"),Weight=0.0167),(Item=(Type=Bone,Name=\"neck_02\"),Weight=0.0585),(Item=(Type=Bone,Name=\"spine_03\"),Weight=0.0723),(Item=(Type=Bone,Name=\"clavicle_r\"),Weight=0.1299),(Item=(Type=Bone,Name=\"clavicle_l\"),Weight=0.1311),(Item=(T ...(truncated)"
    Weight=1
   )
   > &RigUnit_ParentConstraint(
    AdvancedSettings=(RotationOrderForFilter=XZY)
    Child=(Type=Bone,Name="dyn_hair_side_03_r")
    Parents=((Item=(Type=Bone,Name="clavicle_r"),Weight=0.0471),(Item=(Type=Bone,Name="head"),Weight=0.0576),(Item=(Type=Bone,Name="neck_02"),Weight=0.1731),(Item=(Type=Bone,Name="neck_01"),Weight=0.3131),(Item=(Type=Bone,Name="spine_05"),Weight=0.409))
    Weight=1
   )
   > &RigUnit_ParentConstraint(
    AdvancedSettings=(RotationOrderForFilter=XZY)
    Child=(Type=Bone,Name="dyn_hair_front_03_r")
    Parents="((Item=(Type=Bone,Name=\"clavicle_r\"),Weight=0.0101),(Item=(Type=Bone,Name=\"C_jaw\"),Weight=0.0336),(Item=(Type=Bone,Name=\"neck_01\"),Weight=0.1346),(Item=(Type=Bone,Name=\"spine_05\"),Weight=0.1803),(Item=(Type=Bone,Name=\"neck_02\"),Weight=0.3157),(Item=(Type=B ...(truncated)"
    Weight=1
   )
   AlphaInterp = &RigVMFunction_AlphaInterp(InterpSpeedDecreasing=10, InterpSpeedIncreasing=10, Scale=1, bClampResult=False, Value=RigVMVariableNode() {variable=SkydivingFloat}.Value)
   &RigUnit_ModifyTransforms(
    ItemToModify="((Item=(Type=Bone,Name=\"dyn_hair_side_01_r\"),Transform=(Rotation=(X=-0.1105,Y=0.2814,Z=0.102,W=-0.9477),Translation=(X=0,Y=0,Z=0),Scale3D=(X=1,Y=1,Z=1))),(Item=(Type=Bone,Name=\"dyn_hair_back_01\"),Transform=(Rotation=(X=0,Y=0.2003,Z=0,W=-0.9797),Translation ...(truncated)"
    WeightMinimum=0
    Weight=@AlphaInterp.Result
    ExecutePin=@RigVMFunction_Sequence.B
   )
   Multiply = &RigVMFunction_MathDoubleMul(B=RigVMVariableNode() {variable=MoveLeft}.Value, A=RigVMVariableNode() {variable=SkydivingFloat}.Value)
   n1 = &RigVMFunction_AlphaInterp(InterpSpeedDecreasing=10, InterpSpeedIncreasing=10, Scale=1, bClampResult=False, Value=@Multiply.Result)
   > &RigUnit_ModifyTransforms(
    ItemToModify="((Item=(Type=Bone,Name=\"dyn_hair_side_01_r\"),Transform=(Rotation=(X=0,Y=-0.1736,Z=0,W=0.9848),Translation=(X=-2,Y=0,Z=0),Scale3D=(X=1,Y=1,Z=1))),(Item=(Type=Bone,Name=\"None\"),Transform=(Rotation=(X=0,Y=0,Z=0,W=1),Translation=(X=0,Y=0,Z=0),Scale3D=(X=1,Y=1, ...(truncated)"
    WeightMinimum=0
    Weight=@n1.Result
   )
   n2 = &RigVMFunction_AlphaInterp(
    InterpSpeedDecreasing=10
    InterpSpeedIncreasing=10
    Scale=1
    bClampResult=False
    Value=&RigVMFunction_MathDoubleMul(A=RigVMVariableNode() {variable=SkydivingFloat}.Value, B=RigVMVariableNode() {variable=MoveRight}.Value).Result
   )
   > &RigUnit_ModifyTransforms(
    ItemToModify="((Item=(Type=Bone,Name=\"dyn_hair_side_01_l\"),Transform=(Rotation=(X=0,Y=-0.1736,Z=0,W=0.9848),Translation=(X=-2,Y=0,Z=0),Scale3D=(X=1,Y=1,Z=1))),(Item=(Type=Bone,Name=\"dyn_hair_back_02\"),Transform=(Rotation=(X=0,Y=-0.1736,Z=0,W=0.9848),Translation=(X=-2,Y= ...(truncated)"
    WeightMinimum=0
    Weight=@n2.Result
   )
   n3 = &RigVMFunction_AlphaInterp(InterpSpeedDecreasing=10, InterpSpeedIncreasing=10, Scale=1, bClampResult=False, Value=RigVMVariableNode() {variable=ParachuteOpen}.Value)
   > &RigUnit_ModifyTransforms(
    ItemToModify="((Item=(Type=Bone,Name=\"dyn_hair_side_01_r\"),Transform=(Rotation=(X=0,Y=-0.1077,Z=0,W=0.9942),Translation=(X=0,Y=0,Z=0),Scale3D=(X=1,Y=1,Z=1))),(Item=(Type=Bone,Name=\"dyn_hair_back_01\"),Transform=(Rotation=(X=0,Y=-0.0878,Z=0,W=0.9961),Translation=(X=0,Y=0, ...(truncated)"
    WeightMinimum=0
    Weight=@n3.Result
   )
   > n4 = &RigUnit_SphericalPoseReader(
    ActiveRegionScaleFactors=(PositiveWidth=0.5,NegativeWidth=0.5,PositiveHeight=0.5,NegativeHeight=0.5)
    ActiveRegionSize=0.12
    Debug=(bDrawDebug=true,bDraw2D=False,bDrawLocalAxes=False,DebugScale=25,DebugSegments=20,DebugThickness=0.25)
    DriverAxis=(X=1,Y=0,Z=0)
    DriverItem=(Type=Bone,Name="clavicle_l")
    FalloffRegionScaleFactors=(PositiveWidth=0.5,NegativeWidth=0.5,PositiveHeight=0.5,NegativeHeight=0.5)
    FalloffSize=0.3
    FlipHeightScaling=False
    FlipWidthScaling=False
    OptionalParentItem=(Type=Bone,Name="None")
    RotationOffset=(X=0,Y=60,Z=140)
   )
   > &RigUnit_ModifyTransforms(
    ItemToModify="((Item=(Type=Bone,Name=\"dyn_hair_side_01_l\"),Transform=(Rotation=(X=0,Y=-0.166,Z=0,W=0.9861),Translation=(X=0,Y=0,Z=0),Scale3D=(X=1,Y=1,Z=1))),(Item=(Type=Bone,Name=\"dyn_hair_back_01\"),Transform=(Rotation=(X=0,Y=-0.1231,Z=0,W=0.9924),Translation=(X=0,Y=0,Z ...(truncated)"
    WeightMinimum=0
    Weight=&RigVMFunction_MathDoubleMul(A=@n4.OutputParam, B=RigVMVariableNode() {variable=ParachuteOpen}.Value).Result
   )
   > n5 = &RigUnit_SphericalPoseReader(
    ActiveRegionScaleFactors=(PositiveWidth=0.5,NegativeWidth=0.5,PositiveHeight=0.5,NegativeHeight=0.5)
    ActiveRegionSize=0.12
    Debug=(bDrawDebug=true,bDraw2D=False,bDrawLocalAxes=False,DebugScale=25,DebugSegments=20,DebugThickness=0.25)
    DriverAxis=(X=-1,Y=0,Z=0)
    DriverItem=(Type=Bone,Name="clavicle_r")
    FalloffRegionScaleFactors=(PositiveWidth=0.5,NegativeWidth=0.5,PositiveHeight=0.5,NegativeHeight=0.5)
    FalloffSize=0.3
    FlipHeightScaling=False
    FlipWidthScaling=False
    OptionalParentItem=(Type=Bone,Name="None")
    RotationOffset=(X=0,Y=-120,Z=140)
   )
   > &RigUnit_ModifyTransforms(
    ItemToModify="((Item=(Type=Bone,Name=\"dyn_hair_side_01_r\"),Transform=(Rotation=(X=0,Y=-0.2053,Z=0,W=0.9787),Translation=(X=0,Y=0,Z=0),Scale3D=(X=1,Y=1,Z=1))),(Item=(Type=Bone,Name=\"dyn_hair_back_01\"),Transform=(Rotation=(X=0,Y=-0.0781,Z=0,W=0.9969),Translation=(X=0,Y=0, ...(truncated)"
    WeightMinimum=0
    Weight=&RigVMFunction_MathDoubleMul(A=@n5.OutputParam, B=RigVMVariableNode() {variable=ParachuteOpen}.Value).Result
   )
  ]}
]
hierarchy [
 root {loc=<0 0 0>, rot=<0 0 0 1>}
 pelvis {parent=root, loc=<0 -0.7071 93.1191>, rot=<0.0127 0.707 -0.0127 -0.707>}
 spine_01 {parent=pelvis, loc=<6.7494 -0.6388 0>, rot=<0 0 0.034 -0.9994>}
 spine_02 {parent=spine_01, loc=<7.2518 -0.035 0>, rot=<0 0 0.0238 -0.9997>}
 spine_03 {parent=spine_02, loc=<6.0104 0 0>, rot=<0 0 0.0149 -0.9999>}
 spine_04 {parent=spine_03, loc=<8.1024 0 0>, rot=<0 0 -0.0132 -0.9999>}
 spine_05 {parent=spine_04, loc=<10.0146 0 0>, rot=<0 0 -0.0015 -1>}
 neck_01 {parent=spine_05, loc=<11.3138 0.1529 0>, rot=<0 0 -0.2003 -0.9797>}
 neck_02 {parent=neck_01, loc=<4.5627 0 0>, rot=<0 0 0.0504 -0.9987>}
 head {parent=neck_02, loc=<5.3914 0 0>, rot=<0 0 0.115 -0.9934>}
 dyn_hair_base {parent=head, loc=<0 0 0>, rot=<0 0 0 1>}
 dyn_hair_front_01_r {parent=dyn_hair_base, loc=<8.1466 -0.0691 6.7996>, rot=<0.0176 -0.0686 0.9921 -0.1031>}
 dyn_hair_front_02_r {parent=dyn_hair_front_01_r, loc=<6.0756 0 0>, rot=<0 0.0498 0 -0.9988>}
 dyn_hair_front_03_r {parent=dyn_hair_front_02_r, loc=<6.4554 0 0>, rot=<0 0 0 -1>}
 dyn_hair_side_l_base {parent=dyn_hair_base, loc=<9.7607 -5.26 -4.1705>, rot=<0.0842 -0.908 0.4037 -0.0734>}
 dyn_hair_side_01_l {parent=dyn_hair_side_l_base, loc=<0 0 0>, rot=<0 0 0 1>}
 dyn_hair_side_02_l {parent=dyn_hair_side_01_l, loc=<8.4267 0 0>, rot=<0 0.0728 0 -0.9973>}
 dyn_hair_side_03_l {parent=dyn_hair_side_02_l, loc=<7.496 0 -0.2024>, rot=<0 0 0 1>}
 dyn_hair_back_base {parent=dyn_hair_base, loc=<10.184 -7.851 0>, rot=<0.061 -0.7045 0.7045 -0.061>}
 dyn_hair_back_01 {parent=dyn_hair_back_base, loc=<0 0 0>, rot=<0 0 0 1>}
 dyn_hair_back_02 {parent=dyn_hair_back_01, loc=<6.4028 0 0.1311>, rot=<0 0.0863 0 -0.9963>}
 dyn_hair_back_03 {parent=dyn_hair_back_02, loc=<6.4716 0 0>, rot=<0 0.0227 0 -0.9997>}
 dyn_hair_back_04 {parent=dyn_hair_back_03, loc=<7.1668 0 0>, rot=<0 0 0 1>}
 dyn_hair_side_r_base {parent=dyn_hair_base, loc=<9.7606 -5.26 4.1705>, rot=<0.0734 -0.4037 0.908 -0.0842>}
 dyn_hair_side_01_r {parent=dyn_hair_side_r_base, loc=<0 0 0>, rot=<0 0 0 1>}
 dyn_hair_side_02_r {parent=dyn_hair_side_01_r, loc=<8.4266 0 0>, rot=<0 0.0728 0 -0.9973>}
 dyn_hair_side_03_r {parent=dyn_hair_side_02_r, loc=<7.4961 0 -0.2024>, rot=<0 0 0 -1>}
 dyn_hair_front_01_l {parent=dyn_hair_base, loc=<8.1469 -0.0691 -6.7996>, rot=<0.1031 -0.9921 0.0686 -0.0176>}
 dyn_hair_front_02_l {parent=dyn_hair_front_01_l, loc=<6.0755 0 0>, rot=<0 0.0498 0 -0.9988>}
 dyn_hair_front_03_l {parent=dyn_hair_front_02_l, loc=<6.4554 0 0>, rot=<0 0 0 1>}
 dyn_hair_clavicle_side_r_base {parent=dyn_hair_base, loc=<-4.8751 -5.9027 1.7523>, rot=<0.0882 -0.0437 0.9936 -0.0557>}
 dyn_hair_clavicle_side_r {parent=dyn_hair_clavicle_side_r_base, loc=<0 0 0>, rot=<0 0 0 1>}
 clavicle_l {parent=spine_05, loc=<4.3575 -0.1466 -2.3223>, rot=<0.0465 -0.7289 -0.009 -0.6829>}
 clavicle_r {parent=spine_05, loc=<4.3575 -0.1466 2.3223>, rot=<0.7289 0.0465 -0.6829 0.009>}
 upperarm_r {parent=clavicle_r, loc=<-15.5111 0 0>, rot=<0.0226 -0.3234 -0.0088 -0.9459>}
]
variables [
 MoveLeft {type=double, default=0}
 MoveRight {type=double, default=0}
 ParachuteOpen {type=double, default=0}
 Skydiving {type=bool, default=False}
 SkydivingFloat {type=double, default=0}
]
