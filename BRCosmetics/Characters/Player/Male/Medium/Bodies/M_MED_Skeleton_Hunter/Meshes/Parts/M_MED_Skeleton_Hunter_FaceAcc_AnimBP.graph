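# Animation Blueprint for the Skeleton Hunter face accessory, driving its dyn_hair bones from the copied parent pose.
# A bool blend switches between skydive rigid-body physics and a regular chain of look-ats, rigid body and Control Rig.
# Corrective offsets and constraints, a post-sim Control Rig and per-bone AnimDynamics follow, using SkeletonHunterGlitch rigs and physics assets.
# The event graph caches the body mesh and reads player weapon state.
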AnimBlueprint /BRCosmetics/Characters/Player/Male/Medium/Bodies/M_MED_Skeleton_Hunter/Meshes/Parts/M_MED_Skeleton_Hunter_FaceAcc_AnimBP
target_skeleton=&/M_MED_Skeleton_Hunter_FaceAcc_Skeleton
graphs [
 AnimGraph {nodes [
   n1 = RigidBody!(ComponentPose=Local To Component(LocalPose=Use cached pose 'pre_sim'()), Alpha, OverrideWorldGravity:Vector="0,0,0", ExternalForce) -> (Pose) {
    bClampLinearTranslationLimitToRefPose=true
    bOverrideWorldGravity=true
    overridePhysicsAsset=/Game/Characters/Player/Male/Medium/Bodies/M_MED_SkeletonHunterGlitch/Meshes/Parts/M_MED_SkeletonHunterGlitch_FaceAcc_Physics_skydive
    simulationSpace=BaseBoneSpace
    alphaBoolBlend {blendOption=Linear}
    baseBoneRef {boneName=head}
    componentAppliedLinearAccClamp=<2500 2500 2500>
    componentLinearAccScale {z=0.1}
    componentLinearVelScale {z=0.1}
    simSpaceSettings {
     maxAngularAcceleration=500
     maxAngularVelocity=3
     maxLinearAcceleration=400
     maxLinearVelocity=250
     velocityScaleZ=0
     worldAlpha=1
     externalLinearDragV=<0.1 0.1>
    }
   }
   n2 = Component To Local(ComponentPose=@n1)
   n3 = Look At - Bone: dyn_hair_back_base(ComponentPose=Local To Component(LocalPose=Copy Pose From Mesh() {bUseAttachedParent=true}), Alpha=1) {
    alphaBoolBlend {blendOption=Linear}
    boneToModify {boneName=dyn_hair_back_base}
    lookAtLocation=<0 -14>
    lookAtTarget {boneReference {boneName=spine_05}}
    lookAt_Axis {axis=<1 0>}
   }
   n4 = Look At - Bone: dyn_hair_side_l_base(ComponentPose=@n3, Alpha=1) {
    alphaBoolBlend {blendOption=Linear}
    boneToModify {boneName=dyn_hair_side_l_base}
    lookAtLocation=<0 -18 -10>
    lookAtTarget {boneReference {boneName=spine_05}}
    lookAt_Axis {axis=<1 0>}
   }
   n5 = Look At - Bone: dyn_hair_side_r_base(ComponentPose=@n4, Alpha=1) {
    alphaBoolBlend {blendOption=Linear}
    boneToModify {boneName=dyn_hair_side_r_base}
    lookAtLocation=<-26.5545 -19.6734 2.8516>
    lookAtTarget {boneReference {boneName=spine_05}}
    lookAt_Axis {axis=<1 0>}
   }
   n6 = "Transform (Modify) Bone - Bone: dyn_hair_clavicle_side_r"(ComponentPose=@n5, Alpha=1, Translation="-5.420,-1.521,-7.596", Rotation="8.671,0.644,26.494", Scale="0.5,1,1") {
    rotationMode=BMM_Additive
    rotationSpace=BCS_BoneSpace
    scaleMode=BMM_Additive
    scaleSpace=BCS_BoneSpace
    translationMode=BMM_Additive
    translationSpace=BCS_BoneSpace
    alphaBoolBlend {blendOption=Linear}
    boneToModify {boneName=dyn_hair_clavicle_side_r}
    rotation=<rot 8.671 0.644 26.494>
    scale {x=0.5}
    translation=<-5.42 -1.521 -7.596>
   }
   n7 = RigidBody(ComponentPose=@n6, OverrideWorldGravity="0,0,0") {
    bOverrideWorldGravity=true
    overridePhysicsAsset=/Game/Characters/Player/Male/Medium/Bodies/M_MED_SkeletonHunterGlitch/Meshes/Parts/M_MED_SkeletonHunterGlitch_FaceAcc_Physics
    simulationSpace=BaseBoneSpace
    alphaBoolBlend {blendOption=Linear}
    baseBoneRef {boneName=spine_05}
    componentAppliedLinearAccClamp=<2500 2500 2500>
    componentLinearAccScale {z=0.1}
    componentLinearVelScale {z=0.1}
   }
   n8 = "Transform (Modify) Bone - Bone: dyn_hair_clavicle_side_r_base"(ComponentPose=@n7, Alpha=1, Translation="5.641,-2.139,8.687", Rotation="13.681,2.952,-53.484", Scale="1,1,1") {
    rotationMode=BMM_Additive
    rotationSpace=BCS_BoneSpace
    scaleMode=BMM_Additive
    scaleSpace=BCS_BoneSpace
    translationMode=BMM_Additive
    translationSpace=BCS_BoneSpace
    alphaBoolBlend {blendOption=Linear}
    boneToModify {boneName=dyn_hair_clavicle_side_r_base}
    rotation=<rot 13.681 2.952 -53.484>
    translation=<5.641 -2.139 8.6871>
   }
   n9 = "Transform (Modify) Bone - Bone: dyn_hair_side_03_r"(ComponentPose=@n8, Alpha=1, Translation="0,0,0", Rotation="-15.717,-18.711,9.806", Scale="1,1,1") {
    rotationMode=BMM_Additive
    rotationSpace=BCS_BoneSpace
    scaleMode=BMM_Additive
    scaleSpace=BCS_BoneSpace
    translationMode=BMM_Additive
    translationSpace=BCS_BoneSpace
    alphaBoolBlend {blendOption=Linear}
    boneToModify {boneName=dyn_hair_side_03_r}
    rotation=<rot -15.717 -18.711 9.806>
   }
   n10 = "Transform (Modify) Bone - Bone: dyn_hair_front_01_r"(ComponentPose=@n9, Alpha=1, Translation="0,0,0", Rotation="-24.487,23.460,57.040", Scale="1,1,1") {
    rotationMode=BMM_Additive
    rotationSpace=BCS_BoneSpace
    scaleMode=BMM_Additive
    scaleSpace=BCS_BoneSpace
    translationMode=BMM_Additive
    translationSpace=BCS_BoneSpace
    alphaBoolBlend {blendOption=Linear}
    boneToModify {boneName=dyn_hair_front_01_r}
    rotation=<rot -24.487 23.4599 57.04>
   }
   n11 = Control Rig!(
    Source=Component To Local(ComponentPose=@n10)
    Alpha
    Alpha_aim:double=Get Alpha_aim() -> (Alpha_aim:double)
    Alpha_rotate_pawn:double=Get Alpha_rotate_pawn() -> (Alpha_rotate_pawn:double)
    Alpha_pistol:double=Get Alpha_pistol() -> (Alpha_pistol:double)
    Alpha_crouching:double=Get Alpha_crouching() -> (Alpha_crouching:double)
    Alpha_frontend:double=Get Alpha_frontend() -> (Alpha_frontend:double)
    Alpha_moving:double=Get Alpha_moving() -> (Alpha_moving:double)
   ) -> (Pose) {
    alphaBoolBlend {blendOption=Linear}
    controlRigAssetReference {blueprintRigClass=/Game/Characters/Player/Male/Medium/Bodies/M_MED_SkeletonHunterGlitch/Meshes/Parts/M_MED_SkeletonHunterGlitch_FaceAcc_CtrlRig.M_MED_SkeletonHunterGlitch_FaceAcc_CtrlRig_C}
   }
   n12 = "Transform (Modify) Bone - Bone: dyn_hair_side_02_r"(ComponentPose=Local To Component(LocalPose=@n11), Alpha=1, Translation="0,0,0", Rotation="12.849,-18.233,1.577", Scale="1,1,1") {
    rotationMode=BMM_Additive
    rotationSpace=BCS_BoneSpace
    scaleMode=BMM_Additive
    scaleSpace=BCS_BoneSpace
    translationMode=BMM_Additive
    translationSpace=BCS_BoneSpace
    alphaBoolBlend {blendOption=Linear}
    alphaScaleBiasClamp {bInterpResult=true, interpSpeedDecreasing=3, interpSpeedIncreasing=8}
    boneToModify {boneName=dyn_hair_side_02_r}
    rotation=<rot 12.849 -18.233 1.577>
   }
   n13 = "Transform (Modify) Bone - Bone: dyn_hair_back_02"(ComponentPose=@n12, Alpha=1, Translation="0,0,0", Rotation="-21.891,31.960,-22.145", Scale="1,1,1") {
    rotationMode=BMM_Additive
    rotationSpace=BCS_BoneSpace
    scaleMode=BMM_Additive
    scaleSpace=BCS_BoneSpace
    translationMode=BMM_Additive
    translationSpace=BCS_BoneSpace
    alphaBoolBlend {blendOption=Linear}
    alphaScaleBiasClamp {bInterpResult=true, interpSpeedDecreasing=8, interpSpeedIncreasing=8}
    boneToModify {boneName=dyn_hair_back_02}
    rotation=<rot -21.891 31.96 -22.145>
   }
   n14 = "Transform (Modify) Bone - Bone: dyn_hair_back_03"(ComponentPose=@n13, Alpha=1, Translation="0,0,0", Rotation="-32.111,1.931,-18.781", Scale="1,1,1") {
    rotationMode=BMM_Additive
    rotationSpace=BCS_BoneSpace
    scaleMode=BMM_Additive
    scaleSpace=BCS_BoneSpace
    translationMode=BMM_Additive
    translationSpace=BCS_BoneSpace
    alphaBoolBlend {blendOption=Linear}
    alphaScaleBiasClamp {bInterpResult=true, interpSpeedDecreasing=8, interpSpeedIncreasing=8}
    boneToModify {boneName=dyn_hair_back_03}
    rotation=<rot -32.111 1.931 -18.781>
   }
   n15 = "Transform (Modify) Bone - Bone: dyn_hair_back_03"!(ComponentPose=@n14, bAlphaBoolEnabled:bool=True, Translation="0,0,0", Rotation="20.000,-15.024,-0.002", Scale="1,1,1") -> (Pose) {
    alphaInputType=Bool
    rotationMode=BMM_Additive
    rotationSpace=BCS_BoneSpace
    scaleMode=BMM_Additive
    scaleSpace=BCS_BoneSpace
    translationMode=BMM_Additive
    translationSpace=BCS_BoneSpace
    alphaBoolBlend {blendInTime=8, blendOption=Linear, blendOutTime=8}
    boneToModify {boneName=dyn_hair_back_03}
    rotation=<rot 20 -15.024 -0.002>
   }
   n16 = "Transform (Modify) Bone - Bone: dyn_hair_front_01_r"!(ComponentPose=@n15, bAlphaBoolEnabled:bool=True, Translation="0,0,0", Rotation="-5.000,-15.000,0.000", Scale="1,1,1") -> (Pose) {
    alphaInputType=Bool
    rotationMode=BMM_Additive
    rotationSpace=BCS_BoneSpace
    scaleMode=BMM_Additive
    scaleSpace=BCS_BoneSpace
    translationMode=BMM_Additive
    translationSpace=BCS_BoneSpace
    alphaBoolBlend {blendInTime=8, blendOption=Linear, blendOutTime=8}
    boneToModify {boneName=dyn_hair_front_01_r}
    rotation=<rot -5 -15 0>
   }
   n17 = Component To Local(ComponentPose=@n16)
   n18 = "Transform (Modify) Bone - Bone: dyn_hair_back_01"!(
    ComponentPose=Local To Component(LocalPose=Blend Poses by bool(BlendPose_0=@n2, BlendPose_1=@n17, BlendTime_0=1.2, BlendTime_1=0.8) {blendType=Linear})
    bAlphaBoolEnabled:bool=True
    Translation="0,0,0"
    Rotation="26.073,-7.408,9.874"
    Scale="1,1,1"
   ) -> (Pose) {
    alphaInputType=Bool
    rotationMode=BMM_Additive
    rotationSpace=BCS_BoneSpace
    scaleMode=BMM_Additive
    scaleSpace=BCS_BoneSpace
    translationMode=BMM_Additive
    translationSpace=BCS_BoneSpace
    alphaBoolBlend {blendInTime=1.2, blendOption=Linear, blendOutTime=0.8}
    boneToModify {boneName=dyn_hair_back_01}
    rotation=<rot 26.073 -7.408 9.874>
   }
   n19 = "Transform (Modify) Bone - Bone: dyn_hair_back_02"!(ComponentPose=@n18, bAlphaBoolEnabled:bool=True, Translation="0,0,0", Rotation="8.688,-6.044,-8.038", Scale="1,1,1") -> (Pose) {
    alphaInputType=Bool
    rotationMode=BMM_Additive
    rotationSpace=BCS_BoneSpace
    scaleMode=BMM_Additive
    scaleSpace=BCS_BoneSpace
    translationMode=BMM_Additive
    translationSpace=BCS_BoneSpace
    alphaBoolBlend {blendInTime=1.2, blendOption=Linear, blendOutTime=0.8}
    boneToModify {boneName=dyn_hair_back_02}
    rotation=<rot 8.688 -6.044 -8.038>
   }
   n20 = "Transform (Modify) Bone - Bone: dyn_hair_back_03"(ComponentPose=@n19, bAlphaBoolEnabled=True, Translation="0,0,0", Rotation="11.300,21.923,0.579", Scale="1,1,1") {
    alphaInputType=Bool
    rotationMode=BMM_Additive
    rotationSpace=BCS_BoneSpace
    scaleMode=BMM_Additive
    scaleSpace=BCS_BoneSpace
    translationMode=BMM_Additive
    translationSpace=BCS_BoneSpace
    alphaBoolBlend {blendInTime=1.2, blendOption=Linear, blendOutTime=0.8}
    boneToModify {boneName=dyn_hair_back_03}
    rotation=<rot 11.2998 21.923 0.5789>
   }
   n21 = "Transform (Modify) Bone - Bone: dyn_hair_side_01_l"!(ComponentPose=@n20, bAlphaBoolEnabled:bool=false, Translation="0,0,0", Rotation="19.145,-16.891,43.518", Scale="1,1,1") -> (Pose) {
    alphaInputType=Bool
    rotationMode=BMM_Additive
    rotationSpace=BCS_BoneSpace
    scaleMode=BMM_Additive
    scaleSpace=BCS_BoneSpace
    translationMode=BMM_Additive
    translationSpace=BCS_BoneSpace
    alphaBoolBlend {blendInTime=1.2, blendOption=Linear, blendOutTime=0.8}
    alphaScaleBiasClamp {bInterpResult=true, interpSpeedDecreasing=8, interpSpeedIncreasing=8}
    boneToModify {boneName=dyn_hair_side_01_l}
    rotation=<rot 19.1449 -16.891 43.518>
   }
   n22 = "Transform (Modify) Bone - Bone: dyn_hair_side_02_l"!(ComponentPose=@n21, bAlphaBoolEnabled:bool=false, Translation="0,0,0", Rotation="-30.755,0.006,-70.911", Scale="1,1,1") -> (Pose) {
    alphaInputType=Bool
    rotationMode=BMM_Additive
    rotationSpace=BCS_BoneSpace
    scaleMode=BMM_Additive
    scaleSpace=BCS_BoneSpace
    translationMode=BMM_Additive
    translationSpace=BCS_BoneSpace
    alphaBoolBlend {blendInTime=1.2, blendOption=Linear, blendOutTime=0.8}
    alphaScaleBiasClamp {bInterpResult=true, interpSpeedDecreasing=8, interpSpeedIncreasing=8}
    boneToModify {boneName=dyn_hair_side_02_l}
    rotation=<rot -30.755 0.0061 -70.9111>
   }
   n23 = "Transform (Modify) Bone - Bone: dyn_hair_side_01_r"!(ComponentPose=@n22, bAlphaBoolEnabled:bool=false, Translation="0,0,0", Rotation="-8.900,12.442,14.348", Scale="1,1,1") -> (Pose) {
    alphaInputType=Bool
    rotationMode=BMM_Additive
    rotationSpace=BCS_BoneSpace
    scaleMode=BMM_Additive
    scaleSpace=BCS_BoneSpace
    translationMode=BMM_Additive
    translationSpace=BCS_BoneSpace
    alphaBoolBlend {blendInTime=1.2, blendOption=Linear, blendOutTime=0.8}
    alphaScaleBiasClamp {bInterpResult=true, interpSpeedDecreasing=8, interpSpeedIncreasing=8}
    boneToModify {boneName=dyn_hair_side_01_r}
    rotation=<rot -8.9 12.442 14.348>
   }
   n24 = "Transform (Modify) Bone - Bone: dyn_hair_front_02_r"!(ComponentPose=@n23, bAlphaBoolEnabled:bool=false, Translation="0,0,0", Rotation="-16.361,32.582,11.433", Scale="1,1,1") -> (Pose) {
    alphaInputType=Bool
    rotationMode=BMM_Additive
    rotationSpace=BCS_BoneSpace
    scaleMode=BMM_Additive
    scaleSpace=BCS_BoneSpace
    translationMode=BMM_Additive
    translationSpace=BCS_BoneSpace
    alphaBoolBlend {blendInTime=1.2, blendOption=Linear, blendOutTime=0.8}
    alphaScaleBiasClamp {bInterpResult=true, interpSpeedDecreasing=8, interpSpeedIncreasing=8}
    boneToModify {boneName=dyn_hair_front_02_r}
    rotation=<rot -16.361 32.5819 11.433>
   }
   n25 = Get Alpha_moving_and_jump() -> (Alpha_moving_and_jump:double)
   n26 = "Constraint - dyn_hair_back_02 (RR)"!(ComponentPose=@n24, Alpha=@n25, ConstraintWeights_0=0.7, ConstraintWeights_1:float=0.3) -> (Pose) {
    alphaBoolBlend {blendOption=Linear}
    alphaScaleBiasClamp {bInterpResult=true, interpSpeedDecreasing=5, interpSpeedIncreasing=5}
    boneToModify {boneName=dyn_hair_back_02}
    constraintSetup [{transformType=Rotation, targetBone {boneName=dyn_hair_side_02_r}}, {transformType=Rotation, targetBone {boneName=dyn_hair_side_02_l}}]
    constraintWeights [1,1]
   }
   n27 = "Transform (Modify) Bone - Bone: dyn_hair_back_02"!(ComponentPose=@n26, Alpha:float=@n25, Translation="0,0,0", Rotation="-1.281,-16.847,-30.378", Scale="1,1,1") -> (Pose) {
    rotationMode=BMM_Additive
    rotationSpace=BCS_BoneSpace
    scaleMode=BMM_Additive
    scaleSpace=BCS_BoneSpace
    translationMode=BMM_Additive
    translationSpace=BCS_BoneSpace
    alphaBoolBlend {blendInTime=0.3, blendOption=Linear, blendOutTime=0.5}
    alphaScaleBiasClamp {bInterpResult=true, interpSpeedDecreasing=5, interpSpeedIncreasing=5}
    boneToModify {boneName=dyn_hair_back_02}
    rotation=<rot -1.2811 -16.847 -30.378>
   }
   n28 = Get Alpha_moving_and_crouching() -> (Alpha_moving_and_crouching:double)
   n29 = "Transform (Modify) Bone - Bone: dyn_hair_side_03_r"(ComponentPose=@n27, Alpha=@n28, Translation="0,0,0", Rotation="50.496,18.807,-8.755", Scale="1,1,1") {
    rotationMode=BMM_Additive
    rotationSpace=BCS_BoneSpace
    scaleMode=BMM_Additive
    scaleSpace=BCS_BoneSpace
    translationMode=BMM_Additive
    translationSpace=BCS_BoneSpace
    alphaBoolBlend {blendInTime=0.3, blendOption=Linear, blendOutTime=0.5}
    alphaScaleBiasClamp {bInterpResult=true, interpSpeedDecreasing=5, interpSpeedIncreasing=5}
    boneToModify {boneName=dyn_hair_side_03_r}
    rotation=<rot 50.496 18.807 -8.755>
   }
   n30 = "Transform (Modify) Bone - Bone: dyn_hair_back_04"(ComponentPose=@n29, Alpha=@n28, Translation="0,0,0", Rotation="-39.196,-7.983,-9.293", Scale="1,1,1") {
    rotationMode=BMM_Additive
    rotationSpace=BCS_BoneSpace
    scaleMode=BMM_Additive
    scaleSpace=BCS_BoneSpace
    translationMode=BMM_Additive
    translationSpace=BCS_BoneSpace
    alphaBoolBlend {blendInTime=0.3, blendOption=Linear, blendOutTime=0.5}
    alphaScaleBiasClamp {bInterpResult=true, interpSpeedDecreasing=5, interpSpeedIncreasing=5}
    boneToModify {boneName=dyn_hair_back_04}
    rotation=<rot -39.196 -7.983 -9.293>
   }
   n31 = "Constraint - dyn_hair_side_02_l (R)"(ComponentPose=@n30, Alpha=Get Alpha_skydive() -> (Alpha_skydive:double), ConstraintWeights_0=1) {
    alphaBoolBlend {blendOption=Linear}
    boneToModify {boneName=dyn_hair_side_02_l}
    constraintSetup [{transformType=Rotation, targetBone {boneName=dyn_hair_back_02}}]
    constraintWeights [1]
   }
   n32 = "Constraint - dyn_hair_side_02_r (R)"(ComponentPose=@n31, Alpha=Get Alpha_skydive(), ConstraintWeights_0=1) {
    alphaBoolBlend {blendOption=Linear}
    boneToModify {boneName=dyn_hair_side_02_r}
    constraintSetup [{transformType=Rotation, targetBone {boneName=dyn_hair_back_02}}]
    constraintWeights [1]
   }
   n33 = "Transform (Modify) Bone - Bone: dyn_hair_side_02_l"(ComponentPose=@n32, bAlphaBoolEnabled=True, Translation="0,0,0", Rotation="5.253,3.285,-64.441", Scale="1,1,1") {
    alphaInputType=Bool
    rotationMode=BMM_Additive
    rotationSpace=BCS_BoneSpace
    scaleMode=BMM_Additive
    scaleSpace=BCS_BoneSpace
    translationMode=BMM_Additive
    translationSpace=BCS_BoneSpace
    alphaBoolBlend {blendInTime=0.3, blendOption=Linear, blendOutTime=0.5}
    alphaScaleBiasClamp {bInterpResult=true, interpSpeedDecreasing=8, interpSpeedIncreasing=8}
    boneToModify {boneName=dyn_hair_side_02_l}
    rotation=<rot 5.253 3.285 -64.441>
   }
   n34 = "Transform (Modify) Bone - Bone: dyn_hair_side_02_r"!(ComponentPose=@n33, bAlphaBoolEnabled:bool=True, Translation="-0.130,1.307,-0.257", Rotation="-10.583,3.403,68.294", Scale="1,1,1") -> (Pose) {
    alphaInputType=Bool
    rotationMode=BMM_Additive
    rotationSpace=BCS_BoneSpace
    scaleMode=BMM_Additive
    scaleSpace=BCS_BoneSpace
    translationMode=BMM_Additive
    translationSpace=BCS_BoneSpace
    alphaBoolBlend {blendInTime=0.3, blendOption=Linear, blendOutTime=0.5}
    alphaScaleBiasClamp {bInterpResult=true, interpSpeedDecreasing=8, interpSpeedIncreasing=8}
    boneToModify {boneName=dyn_hair_side_02_r}
    rotation=<rot -10.5831 3.4028 68.2942>
    translation=<-0.13 1.307 -0.257>
   }
   n35 = "Transform (Modify) Bone - Bone: dyn_hair_clavicle_side_r"!(ComponentPose=@n34, bAlphaBoolEnabled:bool=True, Translation="-9.976,0.721,1.583", Rotation="53.830,-47.181,20.440", Scale="1,1,1") -> (Pose) {
    alphaInputType=Bool
    rotationMode=BMM_Additive
    rotationSpace=BCS_BoneSpace
    scaleMode=BMM_Additive
    scaleSpace=BCS_BoneSpace
    translationMode=BMM_Additive
    translationSpace=BCS_BoneSpace
    alphaBoolBlend {blendInTime=0.3, blendOption=Linear, blendOutTime=0.5}
    alphaScaleBiasClamp {bInterpResult=true, interpSpeedDecreasing=8, interpSpeedIncreasing=8}
    boneToModify {boneName=dyn_hair_clavicle_side_r}
    rotation=<rot 53.83 -47.181 20.44>
    translation=<-9.976 0.721 1.583>
   }
   n36 = "Constraint - dyn_hair_front_01_r (P)"(ComponentPose=@n35, Alpha=1, ConstraintWeights_0=1) {
    alphaBoolBlend {blendOption=Linear}
    boneToModify {boneName=dyn_hair_front_01_r}
    constraintSetup [{targetBone {boneName=head}}]
    constraintWeights [1]
   }
   n37 = "Transform (Modify) Bone - Bone: dyn_hair_front_01_r"(ComponentPose=@n36, bAlphaBoolEnabled=True, Translation="-0.347,4.112,-1.060", Rotation="25.198,-21.401,3.292", Scale="1,1,1") {
    alphaInputType=Bool
    rotationMode=BMM_Replace
    rotationSpace=BCS_BoneSpace
    scaleMode=BMM_Additive
    scaleSpace=BCS_BoneSpace
    translationMode=BMM_Additive
    translationSpace=BCS_BoneSpace
    alphaBoolBlend {blendInTime=0.3, blendOption=Linear, blendOutTime=0.5}
    alphaScaleBiasClamp {bInterpResult=true, interpSpeedDecreasing=8, interpSpeedIncreasing=8}
    boneToModify {boneName=dyn_hair_front_01_r}
    rotation=<rot 25.198 -21.4011 3.292>
    translation=<-0.347 4.112 -1.06>
   }
   n38 = "Constraint - dyn_hair_front_02_r (P)"(ComponentPose=@n37, Alpha=1, ConstraintWeights_0=1) {
    alphaBoolBlend {blendOption=Linear}
    boneToModify {boneName=dyn_hair_front_02_r}
    constraintSetup [{targetBone {boneName=dyn_hair_front_01_r}}]
    constraintWeights [1]
   }
   n39 = "Transform (Modify) Bone - Bone: dyn_hair_front_02_r"(ComponentPose=@n38, bAlphaBoolEnabled=True, Translation="-0.347,4.112,-1.060", Rotation="0.750,5.308,-25.054", Scale="1,1,1") {
    alphaInputType=Bool
    rotationMode=BMM_Replace
    rotationSpace=BCS_BoneSpace
    scaleMode=BMM_Additive
    scaleSpace=BCS_BoneSpace
    translationMode=BMM_Additive
    translationSpace=BCS_BoneSpace
    alphaBoolBlend {blendInTime=0.3, blendOption=Linear, blendOutTime=0.5}
    alphaScaleBiasClamp {bInterpResult=true, interpSpeedDecreasing=8, interpSpeedIncreasing=8}
    boneToModify {boneName=dyn_hair_front_02_r}
    rotation=<rot 0.75 5.308 -25.054>
    translation=<-0.347 4.112 -1.06>
   }
   n40 = "Transform (Modify) Bone - Bone: dyn_hair_front_02_l"!(ComponentPose=@n39, bAlphaBoolEnabled:bool=True, Translation="1.230,-0.933,3.596", Rotation="-12.848,28.552,-87.720", Scale="1,1,1") -> (Pose) {
    alphaInputType=Bool
    rotationMode=BMM_Additive
    rotationSpace=BCS_BoneSpace
    scaleMode=BMM_Additive
    scaleSpace=BCS_BoneSpace
    translationMode=BMM_Additive
    translationSpace=BCS_BoneSpace
    alphaBoolBlend {blendInTime=0.3, blendOption=Linear, blendOutTime=0.5}
    alphaScaleBiasClamp {bInterpResult=true, interpSpeedDecreasing=8, interpSpeedIncreasing=8}
    boneToModify {boneName=dyn_hair_front_02_l}
    rotation=<rot -12.848 28.552 -87.72>
    translation=<1.23 -0.933 3.596>
   }
   n41 = "Transform (Modify) Bone - Bone: dyn_hair_side_02_l"(ComponentPose=@n40, bAlphaBoolEnabled=True, Translation="1.230,-0.933,3.596", Rotation="-2.678,-0.777,-21.450", Scale="1,1,1") {
    alphaInputType=Bool
    rotationMode=BMM_Additive
    rotationSpace=BCS_BoneSpace
    scaleMode=BMM_Additive
    scaleSpace=BCS_BoneSpace
    translationMode=BMM_Additive
    translationSpace=BCS_BoneSpace
    alphaBoolBlend {blendInTime=0.3, blendOption=Linear, blendOutTime=0.5}
    alphaScaleBiasClamp {bInterpResult=true, interpSpeedDecreasing=8, interpSpeedIncreasing=8}
    boneToModify {boneName=dyn_hair_side_02_l}
    rotation=<rot -2.6781 -0.777 -21.4499>
    translation=<1.23 -0.933 3.596>
   }
   n42 = "Transform (Modify) Bone - Bone: dyn_hair_side_03_l"!(ComponentPose=@n41, bAlphaBoolEnabled:bool=True, Translation="0,0,0", Rotation="-9.999,-0.003,0.000", Scale="1,1,1") -> (Pose) {
    alphaInputType=Bool
    rotationMode=BMM_Additive
    rotationSpace=BCS_BoneSpace
    scaleMode=BMM_Additive
    scaleSpace=BCS_BoneSpace
    translationMode=BMM_Additive
    translationSpace=BCS_BoneSpace
    alphaBoolBlend {blendInTime=0.3, blendOption=Linear, blendOutTime=0.5}
    alphaScaleBiasClamp {bInterpResult=true, interpSpeedDecreasing=8, interpSpeedIncreasing=8}
    boneToModify {boneName=dyn_hair_side_03_l}
    rotation=<rot -9.9991 -0.003 0>
   }
   n43 = "Transform (Modify) Bone - Bone: dyn_hair_back_02"!(ComponentPose=@n42, bAlphaBoolEnabled:bool=True, Translation="0,0,0", Rotation="9.998,0.000,0.000", Scale="1,1,1") -> (Pose) {
    alphaInputType=Bool
    rotationMode=BMM_Additive
    rotationSpace=BCS_BoneSpace
    scaleMode=BMM_Additive
    scaleSpace=BCS_BoneSpace
    translationMode=BMM_Additive
    translationSpace=BCS_BoneSpace
    alphaBoolBlend {blendInTime=0.3, blendOption=Linear, blendOutTime=0.5}
    alphaScaleBiasClamp {bInterpResult=true, interpSpeedDecreasing=8, interpSpeedIncreasing=8}
    boneToModify {boneName=dyn_hair_back_02}
    rotation=<rot 9.9979 0 0>
   }
   n44 = "Transform (Modify) Bone - Bone: dyn_hair_side_02_r"(ComponentPose=@n43, bAlphaBoolEnabled=True, Translation="-0.579,1.657,-8.022", Rotation="17.827,-42.915,70.295", Scale="1,1,1") {
    alphaInputType=Bool
    rotationMode=BMM_Additive
    rotationSpace=BCS_BoneSpace
    scaleMode=BMM_Additive
    scaleSpace=BCS_BoneSpace
    translationMode=BMM_Additive
    translationSpace=BCS_BoneSpace
    alphaBoolBlend {blendInTime=0.3, blendOption=Linear, blendOutTime=0.5}
    alphaScaleBiasClamp {bInterpResult=true, interpSpeedDecreasing=8, interpSpeedIncreasing=8}
    boneToModify {boneName=dyn_hair_side_02_r}
    rotation=<rot 17.827 -42.915 70.295>
    translation=<-0.579 1.657 -8.022>
   }
   n45 = "Transform (Modify) Bone - Bone: dyn_hair_side_03_r"!(ComponentPose=@n44, bAlphaBoolEnabled:bool=True, Translation="0,0,0", Rotation="-15.524,89.900,-36.983", Scale="1,1,1") -> (Pose) {
    alphaInputType=Bool
    rotationMode=BMM_Additive
    rotationSpace=BCS_BoneSpace
    scaleMode=BMM_Additive
    scaleSpace=BCS_BoneSpace
    translationMode=BMM_Additive
    translationSpace=BCS_BoneSpace
    alphaBoolBlend {blendInTime=0.3, blendOption=Linear, blendOutTime=0.5}
    alphaScaleBiasClamp {bInterpResult=true, interpSpeedDecreasing=8, interpSpeedIncreasing=8}
    boneToModify {boneName=dyn_hair_side_03_r}
    rotation=<rot -15.5241 89.8999 -36.983>
   }
   n46 = "Transform (Modify) Bone - Bone: dyn_hair_front_01_r"(ComponentPose=@n45, bAlphaBoolEnabled=True, Translation="0,0,0", Rotation="16.131,20.739,-2.220", Scale="1,1,1") {
    alphaInputType=Bool
    rotationMode=BMM_Additive
    rotationSpace=BCS_BoneSpace
    scaleMode=BMM_Additive
    scaleSpace=BCS_BoneSpace
    translationMode=BMM_Additive
    translationSpace=BCS_BoneSpace
    alphaBoolBlend {blendInTime=0.3, blendOption=Linear, blendOutTime=0.5}
    alphaScaleBiasClamp {bInterpResult=true, interpSpeedDecreasing=8, interpSpeedIncreasing=8}
    boneToModify {boneName=dyn_hair_front_01_r}
    rotation=<rot 16.131 20.739 -2.22>
   }
   n47 = "Transform (Modify) Bone - Bone: dyn_hair_front_02_r"(ComponentPose=@n46, bAlphaBoolEnabled=True, Translation="0,0,0", Rotation="14.338,-9.847,31.517", Scale="0.7,0.7,0.7") {
    alphaInputType=Bool
    rotationMode=BMM_Additive
    rotationSpace=BCS_BoneSpace
    scaleMode=BMM_Additive
    scaleSpace=BCS_BoneSpace
    translationMode=BMM_Additive
    translationSpace=BCS_BoneSpace
    alphaBoolBlend {blendInTime=0.3, blendOption=Linear, blendOutTime=0.5}
    alphaScaleBiasClamp {bInterpResult=true, interpSpeedDecreasing=8, interpSpeedIncreasing=8}
    boneToModify {boneName=dyn_hair_front_02_r}
    rotation=<rot 14.3379 -9.8471 31.517>
    scale=<0.7 0.7 0.7>
   }
   n48 = Control Rig!(
    Source=Component To Local(ComponentPose=@n47)
    Alpha
    SkydivingFloat:double=Get SkydiveFloat() -> (SkydiveFloat:double)
    Skydiving:bool=Get bIsSkyDiving()
    MoveLeft:double="Get Move Left "() -> ("Move Left ":double)
    MoveRight:double=Get Move Right() -> (Move Right:double)
    ParachuteOpen:double=Get ParachuteOpen() -> (ParachuteOpen:double)
   ) -> (Pose) {controlRigAssetReference {blueprintRigClass=/Game/Characters/Player/Male/Medium/Bodies/M_MED_SkeletonHunterGlitch/Meshes/Parts/M_MED_SkeletonHunterGlitch_FaceAcc_Post_Sim_CtrlRig.M_MED_SkeletonHunterGlitch_FaceAcc_Post_Sim_CtrlRig_C}}
   n49 = Anim Dynamics - Bone: dyn_hair_front_01_l!(ComponentPose=Local To Component(LocalPose=@n48), bAlphaBoolEnabled:bool=True, ExternalForce="0,0,0") -> (Pose) {
    alphaInputType=Bool
    angularDampingOverride=1
    angularSpringConstant=20000
    bAngularSpring=true
    bOverrideAngularDamping=true
    bUseGravityOverride=true
    gravityScale=0.35
    simulationSpace=RootRelative
    boundBone {boneName=dyn_hair_front_01_l}
    componentAppliedLinearAccClamp=<2000 2000 2000>
    componentLinearAccScale=<0.5 0.5 0.5>
    componentLinearVelScale=<0.5 0.5 0.5>
    gravityOverride {z=-980}
    physicsBodyDefinitions [{
      boundBone {boneName=dyn_hair_front_01_l}
      boxExtents=<20 30 30>
      constraintSetup {angularLimitsMax=<0 0 5>, angularLimitsMin=<0 -15 -5>, angularTarget=<1 0 0>}
      localJointOffset=<4 0 0>
     }]
   }
   n50 = Anim Dynamics - Bone: dyn_hair_side_01_l!(ComponentPose=@n49, bAlphaBoolEnabled:bool=True, ExternalForce="0,0,0") -> (Pose) {
    alphaInputType=Bool
    angularDampingOverride=1
    angularSpringConstant=20000
    bAngularSpring=true
    bOverrideAngularDamping=true
    bUseGravityOverride=true
    gravityScale=0.35
    simulationSpace=RootRelative
    boundBone {boneName=dyn_hair_side_01_l}
    componentAppliedLinearAccClamp=<2000 2000 2000>
    componentLinearAccScale=<0.5 0.5 0.5>
    componentLinearVelScale=<0.5 0.5 0.5>
    gravityOverride {z=-980}
    physicsBodyDefinitions [{
      boundBone {boneName=dyn_hair_side_01_l}
      boxExtents=<20 30 30>
      constraintSetup {angularLimitsMax=<0 0 5>, angularLimitsMin=<0 -15 -5>, angularTarget=<1 0 0>}
      localJointOffset=<4 0 0>
     }]
   }
   n51 = Anim Dynamics - Bone: dyn_hair_back_01!(ComponentPose=@n50, bAlphaBoolEnabled:bool=True, ExternalForce="0,0,0") -> (Pose) {
    alphaInputType=Bool
    angularDampingOverride=1
    angularSpringConstant=20000
    bAngularSpring=true
    bOverrideAngularDamping=true
    bUseGravityOverride=true
    gravityScale=0.35
    simulationSpace=RootRelative
    boundBone {boneName=dyn_hair_back_01}
    componentAppliedLinearAccClamp=<2000 2000 2000>
    componentLinearAccScale=<0.5 0.5 0.5>
    componentLinearVelScale=<0.5 0.5 0.5>
    gravityOverride {z=-980}
    physicsBodyDefinitions [{
      boundBone {boneName=dyn_hair_back_01}
      boxExtents=<20 30 30>
      constraintSetup {angularLimitsMax=<0 0 5>, angularLimitsMin=<0 -15 -5>, angularTarget=<1 0 0>}
      localJointOffset=<4 0 0>
     }]
   }
   n52 = Anim Dynamics - Bone: dyn_hair_side_01_r!(ComponentPose=@n51, bAlphaBoolEnabled:bool=True, ExternalForce="0,0,0") -> (Pose) {
    alphaInputType=Bool
    angularDampingOverride=1
    angularSpringConstant=20000
    bAngularSpring=true
    bOverrideAngularDamping=true
    bUseGravityOverride=true
    gravityScale=0.35
    simulationSpace=RootRelative
    boundBone {boneName=dyn_hair_side_01_r}
    componentAppliedLinearAccClamp=<2000 2000 2000>
    componentLinearAccScale=<0.5 0.5 0.5>
    componentLinearVelScale=<0.5 0.5 0.5>
    gravityOverride {z=-980}
    physicsBodyDefinitions [{
      boundBone {boneName=dyn_hair_side_01_r}
      boxExtents=<20 30 30>
      constraintSetup {angularLimitsMax=<0 0 5>, angularLimitsMin=<0 -15 -5>, angularTarget=<1 0 0>}
      localJointOffset=<4 0 0>
     }]
   }
   n53 = Anim Dynamics - Bone: dyn_hair_front_01_r!(ComponentPose=@n52, bAlphaBoolEnabled:bool=True, ExternalForce="0,0,0") -> (Pose) {
    alphaInputType=Bool
    angularDampingOverride=1
    angularSpringConstant=20000
    bAngularSpring=true
    bOverrideAngularDamping=true
    bUseGravityOverride=true
    gravityScale=0.35
    simulationSpace=RootRelative
    boundBone {boneName=dyn_hair_front_01_r}
    componentAppliedLinearAccClamp=<2000 2000 2000>
    componentLinearAccScale=<0.5 0.5 0.5>
    componentLinearVelScale=<0.5 0.5 0.5>
    gravityOverride {z=-980}
    physicsBodyDefinitions [{
      boundBone {boneName=dyn_hair_front_01_r}
      boxExtents=<20 30 30>
      constraintSetup {angularLimitsMax=<0 0 5>, angularLimitsMin=<0 -15 -5>, angularTarget=<1 0 0>}
      localJointOffset=<4 0 0>
     }]
   }
   n54 = Anim Dynamics - Bone: dyn_hair_front_02_l!(ComponentPose=@n53, bAlphaBoolEnabled:bool=True, ExternalForce="0,0,0") -> (Pose) {
    alphaInputType=Bool
    angularDampingOverride=1
    angularSpringConstant=20000
    bAngularSpring=true
    bOverrideAngularDamping=true
    bUseGravityOverride=true
    gravityScale=0.35
    simulationSpace=RootRelative
    boundBone {boneName=dyn_hair_front_02_l}
    componentAppliedLinearAccClamp=<2000 2000 2000>
    componentLinearAccScale=<0.5 0.5 0.5>
    componentLinearVelScale=<0.5 0.5 0.5>
    gravityOverride {z=-980}
    physicsBodyDefinitions [{
      boundBone {boneName=dyn_hair_front_02_l}
      boxExtents=<20 30 30>
      constraintSetup {angularLimitsMax=<0 0 5>, angularLimitsMin=<0 -15 -5>, angularTarget=<1 0 0>}
      localJointOffset=<4 0 0>
     }]
   }
   n55 = Anim Dynamics - Bone: dyn_hair_side_02_l!(ComponentPose=@n54, bAlphaBoolEnabled:bool=True, ExternalForce="0,0,0") -> (Pose) {
    alphaInputType=Bool
    angularDampingOverride=1
    angularSpringConstant=20000
    bAngularSpring=true
    bOverrideAngularDamping=true
    bUseGravityOverride=true
    gravityScale=0.35
    simulationSpace=RootRelative
    boundBone {boneName=dyn_hair_side_02_l}
    componentAppliedLinearAccClamp=<2000 2000 2000>
    componentLinearAccScale=<0.5 0.5 0.5>
    componentLinearVelScale=<0.5 0.5 0.5>
    gravityOverride {z=-980}
    physicsBodyDefinitions [{
      boundBone {boneName=dyn_hair_side_02_l}
      boxExtents=<20 30 30>
      constraintSetup {angularLimitsMax=<0 0 5>, angularLimitsMin=<0 -15 -5>, angularTarget=<1 0 0>}
      localJointOffset=<4 0 0>
     }]
   }
   n56 = Anim Dynamics - Bone: dyn_hair_back_02!(ComponentPose=@n55, bAlphaBoolEnabled:bool=True, ExternalForce="0,0,0") -> (Pose) {
    alphaInputType=Bool
    angularDampingOverride=1
    angularSpringConstant=20000
    bAngularSpring=true
    bOverrideAngularDamping=true
    bUseGravityOverride=true
    gravityScale=0.35
    simulationSpace=RootRelative
    boundBone {boneName=dyn_hair_back_02}
    componentAppliedLinearAccClamp=<2000 2000 2000>
    componentLinearAccScale=<0.5 0.5 0.5>
    componentLinearVelScale=<0.5 0.5 0.5>
    gravityOverride {z=-980}
    physicsBodyDefinitions [{
      boundBone {boneName=dyn_hair_back_02}
      boxExtents=<20 30 30>
      constraintSetup {angularLimitsMax=<0 0 5>, angularLimitsMin=<0 -15 -5>, angularTarget=<1 0 0>}
      localJointOffset=<4 0 0>
     }]
   }
   n57 = Anim Dynamics - Bone: dyn_hair_side_02_r!(ComponentPose=@n56, bAlphaBoolEnabled:bool=True, ExternalForce="0,0,0") -> (Pose) {
    alphaInputType=Bool
    angularDampingOverride=1
    angularSpringConstant=20000
    bAngularSpring=true
    bOverrideAngularDamping=true
    bUseGravityOverride=true
    gravityScale=0.35
    simulationSpace=RootRelative
    boundBone {boneName=dyn_hair_side_02_r}
    componentAppliedLinearAccClamp=<2000 2000 2000>
    componentLinearAccScale=<0.5 0.5 0.5>
    componentLinearVelScale=<0.5 0.5 0.5>
    gravityOverride {z=-980}
    physicsBodyDefinitions [{
      boundBone {boneName=dyn_hair_side_02_r}
      boxExtents=<20 30 30>
      constraintSetup {angularLimitsMax=<0 0 5>, angularLimitsMin=<0 -15 -5>, angularTarget=<1 0 0>}
      localJointOffset=<4 0 0>
     }]
   }
   n58 = Anim Dynamics - Bone: dyn_hair_front_02_r!(ComponentPose=@n57, bAlphaBoolEnabled:bool=True, ExternalForce="0,0,0") -> (Pose) {
    alphaInputType=Bool
    angularDampingOverride=1
    angularSpringConstant=20000
    bAngularSpring=true
    bOverrideAngularDamping=true
    bUseGravityOverride=true
    gravityScale=0.35
    simulationSpace=RootRelative
    boundBone {boneName=dyn_hair_front_02_r}
    componentAppliedLinearAccClamp=<2000 2000 2000>
    componentLinearAccScale=<0.5 0.5 0.5>
    componentLinearVelScale=<0.5 0.5 0.5>
    gravityOverride {z=-980}
    physicsBodyDefinitions [{
      boundBone {boneName=dyn_hair_front_02_r}
      boxExtents=<20 30 30>
      constraintSetup {angularLimitsMax=<0 0 5>, angularLimitsMin=<0 -15 -5>, angularTarget=<1 0 0>}
      localJointOffset=<4 0 0>
     }]
   }
   n59 = Anim Dynamics - Bone: dyn_hair_front_03_l!(ComponentPose=@n58, bAlphaBoolEnabled:bool=True, ExternalForce="0,0,0") -> (Pose) {
    alphaInputType=Bool
    angularDampingOverride=1
    angularSpringConstant=20000
    bAngularSpring=true
    bOverrideAngularDamping=true
    bUseGravityOverride=true
    gravityScale=0.35
    simulationSpace=RootRelative
    boundBone {boneName=dyn_hair_front_03_l}
    componentAppliedLinearAccClamp=<2000 2000 2000>
    componentLinearAccScale=<0.5 0.5 0.5>
    componentLinearVelScale=<0.5 0.5 0.5>
    gravityOverride {z=-980}
    physicsBodyDefinitions [{
      boundBone {boneName=dyn_hair_front_03_l}
      boxExtents=<20 30 30>
      constraintSetup {angularLimitsMax=<0 0 5>, angularLimitsMin=<0 -15 -5>, angularTarget=<1 0 0>}
      localJointOffset=<4 0 0>
     }]
   }
   n60 = Anim Dynamics - Bone: dyn_hair_side_03_l!(ComponentPose=@n59, bAlphaBoolEnabled:bool=True, ExternalForce="0,0,0") -> (Pose) {
    alphaInputType=Bool
    angularDampingOverride=1
    angularSpringConstant=20000
    bAngularSpring=true
    bOverrideAngularDamping=true
    bUseGravityOverride=true
    gravityScale=0.35
    simulationSpace=RootRelative
    boundBone {boneName=dyn_hair_side_03_l}
    componentAppliedLinearAccClamp=<2000 2000 2000>
    componentLinearAccScale=<0.5 0.5 0.5>
    componentLinearVelScale=<0.5 0.5 0.5>
    gravityOverride {z=-980}
    physicsBodyDefinitions [{
      boundBone {boneName=dyn_hair_side_03_l}
      boxExtents=<20 30 30>
      constraintSetup {angularLimitsMax=<0 0 5>, angularLimitsMin=<0 -15 -5>, angularTarget=<1 0 0>}
      localJointOffset=<4 0 0>
     }]
   }
   n61 = Anim Dynamics - Bone: dyn_hair_back_04!(ComponentPose=@n60, bAlphaBoolEnabled:bool=True, ExternalForce="0,0,0") -> (Pose) {
    alphaInputType=Bool
    angularDampingOverride=1
    angularSpringConstant=20000
    bAngularSpring=true
    bOverrideAngularDamping=true
    bUseGravityOverride=true
    gravityScale=0.35
    simulationSpace=RootRelative
    boundBone {boneName=dyn_hair_back_04}
    componentAppliedLinearAccClamp=<2000 2000 2000>
    componentLinearAccScale=<0.5 0.5 0.5>
    componentLinearVelScale=<0.5 0.5 0.5>
    gravityOverride {z=-980}
    physicsBodyDefinitions [{
      boundBone {boneName=dyn_hair_back_04}
      boxExtents=<20 30 30>
      constraintSetup {angularLimitsMax=<0 0 5>, angularLimitsMin=<0 -15 -5>, angularTarget=<1 0 0>}
      localJointOffset=<4 0 0>
     }]
   }
   n62 = Anim Dynamics - Bone: dyn_hair_side_03_r!(ComponentPose=@n61, bAlphaBoolEnabled:bool=True, ExternalForce="0,0,0") -> (Pose) {
    alphaInputType=Bool
    angularDampingOverride=1
    angularSpringConstant=20000
    bAngularSpring=true
    bOverrideAngularDamping=true
    bUseGravityOverride=true
    gravityScale=0.35
    simulationSpace=RootRelative
    boundBone {boneName=dyn_hair_side_03_r}
    componentAppliedLinearAccClamp=<2000 2000 2000>
    componentLinearAccScale=<0.5 0.5 0.5>
    componentLinearVelScale=<0.5 0.5 0.5>
    gravityOverride {z=-980}
    physicsBodyDefinitions [{
      boundBone {boneName=dyn_hair_side_03_r}
      boxExtents=<20 30 30>
      constraintSetup {angularLimitsMax=<0 0 5>, angularLimitsMin=<0 -15 -5>, angularTarget=<1 0 0>}
      localJointOffset=<4 0 0>
     }]
   }
   n63 = Anim Dynamics - Bone: dyn_hair_front_03_r!(ComponentPose=@n62, bAlphaBoolEnabled:bool=True, ExternalForce="0,0,0") -> (Pose) {
    alphaInputType=Bool
    angularDampingOverride=1
    angularSpringConstant=20000
    bAngularSpring=true
    bOverrideAngularDamping=true
    bUseGravityOverride=true
    gravityScale=0.35
    simulationSpace=RootRelative
    boundBone {boneName=dyn_hair_front_03_r}
    componentAppliedLinearAccClamp=<2000 2000 2000>
    componentLinearAccScale=<0.5 0.5 0.5>
    componentLinearVelScale=<0.5 0.5 0.5>
    gravityOverride {z=-980}
    physicsBodyDefinitions [{
      boundBone {boneName=dyn_hair_front_03_r}
      boxExtents=<20 30 30>
      constraintSetup {angularLimitsMax=<0 0 5>, angularLimitsMin=<0 -10 -5>, angularTarget=<1 0 0>}
      localJointOffset=<4 0 0>
     }]
   }
   Output Pose(Result=Component To Local(ComponentPose=@n63)) {layerGroup=None}
   n64 = Get BodyMesh() -> (BodyMesh:Skeletal Mesh Component Object Reference)
   Save cached pose 'pre_sim'(Pose=Layered blend per bone(
     BasePose=Copy Pose From Mesh() {bUseAttachedParent=true}
     BlendPoses_0=Copy Pose From Mesh(SourceMeshComponent=@n64) {bUseAttachedParent=true}
    ) {layerSetup [{branchFilters [{boneName=upperarm_l}, {boneName=upperarm_r}]}]}) {cacheName=pre_sim}
  ]}
 EventGraph {nodes [
   Event BlueprintInitializeAnimation()
   > n1 = Cast To PlayerPawn_Athena(Object=TryGetPawnOwner())
   > Set BodyMesh(BodyMesh=GetSkeletalMeshForPartType(self=@n1))
   n2 = Event BlueprintUpdateAnimation()
   > n3 = Sequence!(execute) -> (then_0, then_1, then_2:Exec)
   n4 = Sequence(execute=@n3.then_2)
   n5 = Get Timer_switch_weapon() -> (Timer_switch_weapon:double)
   > n6 = Branch(Condition="Equal (Float)"(A=@n5))
   > Set Alpha_switch_weapon(Alpha_switch_weapon:double=0) -> (Output_Get:double)
   Set Timer_switch_weapon(
    execute=@n6.else
    Timer_switch_weapon:double=FInterpTo_Constant!(Current:double=@n5, Target:double=0, DeltaTime:double=@n2.DeltaTimeX, InterpSpeed:double=0.185) -> (ReturnValue:double)
   ) -> (Output_Get:double)
   n7 = GetAnimInstance(self=GetOwnerSkeletalMeshComponent())
   n8 = Cast To FortPlayerAnimInstance(execute=@n4.then_1, Object=@n7)
   n9 = Get WeaponCoreAnim(self=@n8)
   n10 = Set Current Weapon(execute=@n12.then, Current Weapon:EFortWeaponCoreAnimation Enum=@n9) -> (Output_Get:EFortWeaponCoreAnimation Enum)
   n11 = Branch(execute=@n8.then, Condition="Equal (Byte)"!(A:Byte=@n10, B:Byte=@n9) -> (ReturnValue:bool))
   Set Alpha_switch_weapon(execute=@n11.else, Alpha_switch_weapon=1)
   > n12 = Set Timer_switch_weapon(Timer_switch_weapon=1)
   Is Valid(exec=@n4.then_2, InputObject=@n7)
   > Set Is Custom Emote Playing(Is Custom Emote Playing:bool=Montage_IsPlaying(self=@n7, Montage=/Game/Animation/Game/MainPlayer/Emotes/Emote_Celebration_M)) -> (Output_Get:bool)
   > n13 = PlayerMovement(AnimInstance=Self-Reference(), XDotProductCoefficient=1, YDotProductCoefficient=1, ZDotProductCoefficient=1)
   n14 = BreakVector(InVec=MakeVector(X=@n13, Y=@n13.YMovement))
   > "Set Move Left "("Move Left ":double="Absolute (Float)"(A="Clamp (Float)"(Value=@n14, Min=-1, Max=0))) -> (Output_Get:double)
   > Set Move Right(Move Right:double="Clamp (Float)"(Value=@n14, Min=0, Max=1)) -> (Output_Get:double)
   > Set moveForward(moveForward:double="Clamp (Float)"(Value=@n14.Y, Min=0, Max=1)) -> (Output_Get:double)
   > Set moveBackward(moveBackward:double="Absolute (Float)"(A="Clamp (Float)"(Value=@n14.Y, Min=-1, Max=0))) -> (Output_Get:double)
   n15 = Gravity Override State Function DEPRECATED(
    execute=@n3.then_0
    CharacterPartAnimInstance=Self-Reference()
    GameTimeInSeconds=GetGameTimeInSeconds()
    GravityOverrideParams=Get Gravity Override Params()
   )
   > Set gravity_override(gravity_override=@n15)
   Set Alpha_skydive(execute=@n3.then_1, Alpha_skydive:double="To Float (Boolean)"(InBool=Get bIsSkyDiving())) -> (Output_Get:double)
   n16 = Get bIsJumping()
   n17 = Get bIsFalling()
   > Set Alpha_jumping(Alpha_jumping:double="To Float (Boolean)"(InBool=OR Boolean(A=@n16, B=@n17))) -> (Output_Get:double)
   > Set Alpha_aim(Alpha_aim:double="To Float (Boolean)"(InBool=Get bIsTargeting())) -> (Output_Get:double)
   > Set Alpha_crouching(Alpha_crouching:double="To Float (Boolean)"(InBool=Get bIsCrouching())) -> (Output_Get:double)
   > Set Alpha_frontend(Alpha_frontend:double="To Float (Boolean)"(InBool=Get bIsInFrontEnd())) -> (Output_Get:double)
   > Set Alpha_moving(Alpha_moving:double="To Float (Boolean)"(InBool=Get bIsMoving2D())) -> (Output_Get:double)
   n18 = NOT Boolean(A=Get bIsCrouching())
   n19 = OR Boolean!(A=Get bIsMoving2D(), B=Get bIsJumping(), C:bool=Get bIsFalling()) -> (ReturnValue)
   > Set Alpha_moving_and_jump(Alpha_moving_and_jump:double="To Float (Boolean)"(InBool=AND Boolean(A=@n19, B=@n18))) -> (Output_Get:double)
   > Set SkydiveFloat(SkydiveFloat:double="To Float (Boolean)"(InBool=Get bIsSkydiveFloating())) -> (Output_Get:double)
   n20 = Get bIsMoving2D()
   n21 = Get bIsCrouching()
   > Set Alpha_moving_and_crouching(Alpha_moving_and_crouching:double="To Float (Boolean)"(InBool=AND Boolean(A=@n21, B=@n20))) -> (Output_Get:double)
   > Set Falling_or_skydive(Falling_or_skydive:bool=OR Boolean!(A=Get bIsFalling(), B=Get bIsSkyDiving()) -> (ReturnValue)) -> (Output_Get:bool)
   n22 = GetWeaponCoreAnim!(self:Fort Player Animation Interface Object Reference=GetOwnerAnimationInterface!(self:Custom Character Part Anim Instance Object Reference) -> (ReturnValue:Fort Player Animation Interface Object Reference)) -> (ReturnValue:EFortWeaponCoreAnimation Enum)
   > Set Alpha_pistol(Alpha_pistol:double="To Float (Boolean)"(InBool="Equal (Enum)"(A:EFortWeaponCoreAnimation Enum=@n22, B:EFortWeaponCoreAnimation Enum=Pistol))) -> (Output_Get:double)
   n23 = Get!() -> (PawnDeltaRotation_Roll:float, PawnDeltaRotation_Pitch:float, PawnDeltaRotation_Yaw:float)
   > Set Alpha_rotate_pawn(Alpha_rotate_pawn:double=MapRangeClamped(Value="Absolute (Float)"(A=@n23.PawnDeltaRotation_Yaw), InRangeA=1, InRangeB=5)) -> (Output_Get:double)
   > Set ParachuteOpen(ParachuteOpen:double="To Float (Boolean)"(InBool=Get bIsParachuteOpened())) -> (Output_Get:double)
  ]}
]
variables [
 Alpha_aim {type=number}
 Alpha_crouching {type=number}
 Alpha_frontend {type=number}
 Alpha_jumping {type=number}
 Alpha_moving {type=number}
 Alpha_moving_and_crouching {type=number}
 Alpha_moving_and_jump {type=number}
 Alpha_pistol {type=number}
 Alpha_relaxed {type=number}
 Alpha_rotate_pawn {type=number}
 Alpha_skydive {type=number}
 Alpha_switch_weapon {type=number}
 BodyMesh {type=SkeletalMeshComponent}
 Current Weapon {type=EFortWeaponCoreAnimation, default=Melee}
 Falling_or_skydive {type=boolean}
 Gravity Override Params {
  type=GravityOverrideParamsStruct
  default {
   onGround_Falling {gravityOverride=<0 0 900>}
   skydive_DivingDown {gravityOverride=<0 500 0>, windAmplitude=<50 800 100>}
   skydive_FloatingDown {gravityOverride=<0 0 -980>, windAmplitude=<0 0 0>, windOffset=<0 0 0>}
   skydive_GoingUp {gravityOverride=<0 0 -980>, windAmplitude=<0 0 0>}
   skydive_Parachuting {gravityOverride=<200 100 -400>, windAmplitude=<50 -125 250>, windFrequency=<0.75 3.5 4>}
   tactical_Sprint {gravityOverride=<0 0 200>}
  }
 }
 gravity_override {type=Vector}
 Is Custom Emote Playing {type=boolean}
 "Move Left " {type=number}
 Move Right {type=number}
 moveBackward {type=number}
 moveForward {type=number}
 ParachuteOpen {type=number}
 SkydiveFloat {type=number}
 Timer_switch_weapon {type=number}
]
